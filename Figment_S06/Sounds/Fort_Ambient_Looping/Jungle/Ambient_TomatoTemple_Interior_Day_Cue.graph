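# Looping ambient sound cue for the daytime interior of the Tomato Temple in the Figment_S06 jungle ambience set.
# A quality level node routes its first two levels to one looping wave player and leaves the remaining two slots empty.
# The cue plays at reduced volume through the tile ambient attenuation and the ambient tile mix sound class.
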SoundCue /Figment_S06/Sounds/Fort_Ambient_Looping/Jungle/Ambient_TomatoTemple_Interior_Day_Cue
root=QualityLevel_0
nodes [
 {id=QualityLevel_0, children [WavePlayer_1,WavePlayer_1,null,null]}
 {
  id=WavePlayer_1
  properties {bLooping=true, soundWaveAssetPtr=&/Ambient_TomatoTemple_Interior_Day_01}
 }
]
properties {
 attenuationSettings=/Game/Sounds/Attenuation/Fort_TileAmbient_Att
 maxDistance=10000
 soundClassObject=/Game/Sounds/FortSoundClassesAndModes/Fort_Ambient_TileMix
 volumeMultiplier=0.6
}
